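# Sound class that groups storm-related audio for the Figment_S09 content, nested under the in-game SFX class.
# It parents two child classes from the base game folder, one for storms (Fort_Storms_PSM) and one for enemy spawns.
# Its only property override is a small LFE bleed.
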SoundClass /Figment_S09/Sounds/FortSoundClassesAndModes/Fort_10_Storm_Related
parent=&Fort_SFX_InGame
children [/Game/Sounds/FortSoundClassesAndModes/Fort_Storms_PSM, /Game/Sounds/FortSoundClassesAndModes/Fort_Enemy_Spawn]
properties {properties {lFEBleed=0.2}}
