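# Sound class for player searching foley, by its name the sounds heard while searching or looting in Figment_S09.
# It sits under the Player_Searching_looting_crafting parent class in the Fort sound class hierarchy.
# It boosts volume, applies effects and bleeds a portion into the LFE channel.
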SoundClass /Figment_S09/Sounds/FortSoundClassesAndModes/Fort_Searching_Foley
parent=&/Fort_08_Player_Searching_looting_crafting
properties {properties {bApplyEffects=true, lFEBleed=0.2, volume=2}}
