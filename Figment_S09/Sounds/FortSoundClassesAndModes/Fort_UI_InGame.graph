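# Sound class for in-game UI sounds in the Figment_S09 content, parented to the Fort_04_Normal_Announcements class.
# It flags its sounds as UI audio, turns reverb off and bleeds a portion of the signal into the LFE channel.
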SoundClass /Figment_S09/Sounds/FortSoundClassesAndModes/Fort_UI_InGame
parent=/Game/Sounds/FortSoundClassesAndModes/Fort_04_Normal_Announcements
properties {properties {bIsUISound=true, bReverb=false, lFEBleed=0.25}}
