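# Fort_Player_Loot_Low is a Sound Class in the Figment_S09 audio hierarchy, named for low-tier player loot sounds.
# It is a child of Fort_Loot_Drop_Collect_Suppress, placing it in the loot drop and collect branch.
# Its own properties enable effects processing and set a partial LFE bleed.
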SoundClass /Figment_S09/Sounds/FortSoundClassesAndModes/Fort_Player_Loot_Low
parent=&/Fort_Loot_Drop_Collect_Suppress
properties {properties {bApplyEffects=true, lFEBleed=0.25}}
