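# Sound class for suppressing loot drop and collect sounds, parented to the player searching, looting and crafting class.
# It has Fort_Player_Loot_Low as its child class and sets slightly reduced volume with some LFE bleed.
# A passive sound mix modifier triggers the Fort_Large_Loot_Drop_Collect_Suppress mix while sounds in this class play, with no minimum volume threshold.
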SoundClass /Figment_S09/Sounds/FortSoundClassesAndModes/Fort_Loot_Drop_Collect_Suppress
parent=&/Fort_08_Player_Searching_looting_crafting
children [&/Fort_Player_Loot_Low]
properties {
 passiveSoundMixModifiers [{minVolumeThreshold=0, soundMix=&/Fort_Large_Loot_Drop_Collect_Suppress}]
 properties {lFEBleed=0.25, volume=0.8}
}
