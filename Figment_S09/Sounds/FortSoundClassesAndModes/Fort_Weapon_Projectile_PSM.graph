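# Sound class for weapon projectile audio in Figment Season 9, a child of Fort_WeaponProjectile.
# It passively activates the Fort_Weapon_Duck_MixMod sound mix while its sounds play above a minimum volume threshold, which by name ducks other audio.
# The class also boosts volume, applies effects and sets an LFE bleed.
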SoundClass /Figment_S09/Sounds/FortSoundClassesAndModes/Fort_Weapon_Projectile_PSM
parent=&/Fort_WeaponProjectile
properties {
 passiveSoundMixModifiers [{minVolumeThreshold=0.1, soundMix=&/Fort_Weapon_Duck_MixMod}]
 properties {bApplyEffects=true, lFEBleed=0.25, volume=1.3}
}
