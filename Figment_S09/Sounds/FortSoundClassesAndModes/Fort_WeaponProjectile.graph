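# Sound class that groups weapon projectile audio for the Figment S09 content, nested under the base game's Fort_Weapons class.
# It branches into six child classes for projectile, distant, first-person, projectile effects, and low-ammo degraded and critical sounds, several with a PSM suffix.
# Its own settings enable effects, bleed some signal to the LFE channel and set volume slightly below unity.
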SoundClass /Figment_S09/Sounds/FortSoundClassesAndModes/Fort_WeaponProjectile
parent=/Game/Sounds/FortSoundClassesAndModes/Fort_Weapons
children [
 &/Fort_Weapon_Projectile_PSM
 &/Fort_Weapon_Projectile_Distant
 &/Fort_ProjectileEffects
 &/Fort_Weapon_LowAmmoDegraded_PSM
 &/Fort_Weapon_LowAmmoCritical_PSM
 &/Fort_Weapon_Projectile_FirstPerson_PSM
]
properties {properties {bApplyEffects=true, lFEBleed=0.25, volume=0.8}}
